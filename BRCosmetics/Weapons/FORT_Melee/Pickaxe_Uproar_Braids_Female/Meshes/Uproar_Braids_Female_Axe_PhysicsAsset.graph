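# Physics asset for the Uproar Braids Female pickaxe, defining collision bodies and joint limits for its handle and dyn_ bones.
# A long capsule covers the main two-handed handle, with a sphere for the hammerhead and capsules for the tube and two handle segments.
# Tube, hammerhead and master handle segment constrain to the main handle, while the second segment chains off the master.
# Tight angular limits allow slight wobble, with the tube swinging most freely.
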PhysicsAsset /BRCosmetics/Weapons/FORT_Melee/Pickaxe_Uproar_Braids_Female/Meshes/Uproar_Braids_Female_Axe_PhysicsAsset
bodies {
 Melee_TwoHanded_Handle [{shapeType=Capsule, length=141.3517, radius=4.9638, center=<1.1542 -2.2733 33.784>, rotation=<rot -1.3943 -90 0.002>}]
 dyn_hammerhead [{shapeType=Sphere, radius=16.3851, shapeName=dyn_hammerhead_sphere, center=<1.1626 -0.3684 0.1>}]
 dyn_handle_01 [{
   shapeType=Capsule
   length=12.4046
   radius=2.25
   shapeName=dyn_handle_01_capsule
   center=<6.6608 -0.3786 1.1256>
   rotation=<rot 0 -94.1033 -89.7503>
  }]
 dyn_handle_MSTR [{
   shapeType=Capsule
   length=9.8192
   radius=1.8012
   shapeName=dyn_handle_MSTR_capsule
   center=<4.4523 -0.0238 1.1621>
   rotation=<rot 0 -89.4802 -89.9936>
  }]
 dyn_tube [{shapeType=Capsule, length=11.505, radius=7.005, shapeName=dyn_tube_capsule, center=<0.9398 -6.4814 -0.0822>, rotation=<rot -5.0002 0 0>}]
}
constraints [
 {bone1Name=dyn_tube, bone2Name=Melee_TwoHanded_Handle, swing1LimitDegrees=60, swing2LimitDegrees=15, twistLimitDegrees=0.1}
 {bone1Name=dyn_handle_MSTR, bone2Name=Melee_TwoHanded_Handle, swing1LimitDegrees=8, swing2LimitDegrees=0.001, twistLimitDegrees=0.001}
 {bone1Name=dyn_handle_01, bone2Name=dyn_handle_MSTR, swing1LimitDegrees=3, swing2LimitDegrees=0.001, twistLimitDegrees=0.001}
 {bone1Name=dyn_hammerhead, bone2Name=Melee_TwoHanded_Handle, swing1LimitDegrees=0.1, swing2LimitDegrees=2.5, twistLimitDegrees=1}
]
