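# Animation Blueprint for the Uproar Braids Female axe pickaxe that adds secondary motion to the weapon mesh.
# A RigidBody node simulates in handle space using an override physics asset, then AnimDynamics springs the dyn_hammerhead bone relative to the handle.
# An additive bone offset then shifts the hammerhead, with alpha from Is_Swinging.
# Each update sets that value to whether the owning player's anim instance is playing the PlayerMeleeCombo montage.
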AnimBlueprint /BRCosmetics/Weapons/FORT_Melee/Pickaxe_Uproar_Braids_Female/Meshes/Uproar_Braids_Female_Axe_AnimBP
target_skeleton=&/Uproar_Braids_Female_Axe_Skeleton
graphs [
 AnimGraph {nodes [
   n1 = RigidBody(ComponentPose=Local To Component(LocalPose=Input Pose())) {
    bClampLinearTranslationLimitToRefPose=true
    bOverrideWorldGravity=true
    overridePhysicsAsset=&/Uproar_Braids_Female_Axe_PhysicsAsset
    simulationSpace=BaseBoneSpace
    alphaBoolBlend {blendOption=Linear}
    baseBoneRef {boneName=Melee_TwoHanded_Handle}
    componentAppliedLinearAccClamp=<3000 3000 3000>
    componentLinearAccScale=<0.4 0.4 0.3>
    componentLinearVelScale=<0.4 0.4 0.3>
    overrideWorldGravity {z=-980}
   }
   n2 = Anim Dynamics - Bone: dyn_hammerhead(ComponentPose=@n1, Alpha=0.8, ExternalForce="0,0,0") {
    angularDampingOverride=0.9
    angularSpringConstant=125
    bAngularSpring=true
    bGravityOverrideInSimSpace=true
    bLinearSpring=true
    bOverrideAngularDamping=true
    bOverrideLinearDamping=true
    gravityScale=0
    linearDampingOverride=0.995
    linearSpringConstant=157
    simulationSpace=BoneRelative
    alphaBoolBlend {blendOption=Linear}
    boundBone {boneName=dyn_hammerhead}
    componentAppliedLinearAccClamp=<6000 6000 6000>
    componentLinearAccScale=<0.3 0.3 0.1>
    componentLinearVelScale=<0.7 0.7 0.8>
    gravityOverride {z=-980}
    physicsBodyDefinitions [{
      sphereCollisionRadius=0
      boundBone {boneName=dyn_hammerhead}
      boxExtents=<5 5 5>
      constraintSetup {
       angularTargetAxis=AxisZ
       twistAxis=AxisZ
       angularLimitsMax=<0 0 0>
       angularLimitsMin=<0 0 0>
       angularTarget=<0 0 0>
       linearAxesMax=<0 0.8 0.2>
       linearAxesMin=<0 -1.5 -0.2>
      }
      localJointOffset=<-0.1 -0.2 0>
     }]
    relativeSpaceBone {boneName=Melee_TwoHanded_Handle}
   }
   n3 = "Transform (Modify) Bone - Bone: dyn_hammerhead"(ComponentPose=@n2, Alpha=Get Is_Swinging() -> (Is_Swinging:double), Translation="0.000,-1.2,0.000", Rotation="0,0,0", Scale="1,1,1") {
    rotationMode=BMM_Additive
    translationMode=BMM_Additive
    alphaBoolBlend {blendOption=Linear}
    boneToModify {boneName=dyn_hammerhead}
    translation {y=-1.2}
   }
   Output Pose(Result=Component To Local(ComponentPose=@n3)) {layerGroup=None}
  ]}
 EventGraph {nodes [
   Event BlueprintUpdateAnimation()
   n1 = GetOwnerFortPlayerAnimInstance!(self:Custom Character Part Anim Instance Object Reference) -> (ReturnValue:Fort Player Anim Instance Object Reference)
   > Set Is_Swinging(Is_Swinging:double="To Float (Boolean)"(InBool=Montage_IsPlaying(self=@n1, Montage=/Game/Animation/Game/MainPlayer/Montages/Melee/PlayerMeleeCombo))) -> (Output_Get:double)
   TryGetPawnOwner()
  ]}
]
variables [
 Is_Swinging {type=number}
 Pulse {type=Vector}
 Shake {type=Rotator}
]
